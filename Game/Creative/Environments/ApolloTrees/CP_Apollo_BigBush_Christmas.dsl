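; Christmas big bush environment Blueprint that reuses cornfield logic to react to players and vehicles.
; Server enables sphere overlaps; entering players receive the cornfield GameplayEffect and ignore their next fall damage, and exiting removes the effect.
; If the Cornfield.Wobble curve-table value allows, entry drives a cosmetic dynamic-material wobble from player location and velocity, stronger at higher speed.
; A vehicle overlap destroys the bush.
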(event EventBeginPlay
  ;[1 1 1] EVENT -> BeginPlay
  (bind _outxy (DataTable|EvaluateCurveTableRow "/Game/Athena/Balance/DataTables/AthenaGameData.AthenaGameData" "Default.Cornfield.Wobble")
    (:RowFound
      (Variables|Default|SetShouldWobble (Math|Conversions|ToBoolean(Integer) (Math|Float|Truncate _outxy))))
    (:RowNotFound
      _))
  (if (Networking|IsServer)
    (Variables|/Script/Engine.PrimitiveComponent|SetbGenerateOverlapEvents true (Variables|Default|GetOverlapSphere))))

(event Custom|MaterialWobble
  (params (player :type Actor%20Object%20Reference))
  (bind _mid (Variables|Default|GetMID))
  ;[1 1 1] Material wobble
  (Networking|CanExecuteCosmeticEvents
    (:True
      (Utilities|IsValid (Variables|Default|GetMID)
        (:Is%20Valid
          _)
        (:Is%20Not%20Valid
          (bind _returnvalue (Rendering|Material|CreateDynamicMaterialInstance (Variables|Debug|GetStaticMeshComponent) 0 0 "None" :!class /Script/Engine.PrimitiveComponent))
          (Variables|Default|SetMID _returnvalue)))
      (bind _asplayer_pawn_athena (Utilities|Casting|CastToPlayerPawn_Athena player))
      (Variables|Default|SetPlayerVelocity (Transformation|GetVelocity _asplayer_pawn_athena))
      (Rendering|Material|SetVectorParameterValue _mid "HitLocation" (Math|Conversions|ToLinearColor(Vector) (Transformation|GetActorLocation _asplayer_pawn_athena)) :!class /Script/Engine.MaterialInstanceDynamic)
      (Rendering|Material|SetVectorParameterValue _mid "WobbleDirection" (Math|Conversions|ToLinearColor(Vector) (Math|Vector|Normalize (Variables|Default|GetPlayerVelocity))) :!class /Script/Engine.MaterialInstanceDynamic)
      (Rendering|Material|SetScalarParameterValue _mid "Time" (Utilities|Time|GetTimeSeconds) :!class /Script/Engine.MaterialInstanceDynamic)
      (Rendering|Material|SetScalarParameterValue _mid "WobbleStrength" (select (> (Math|Vector|VectorLength (Variables|Default|GetPlayerVelocity)) 295.0) 3e+01 5.0) :!class /Script/Engine.MaterialInstanceDynamic))))

(event OnComponentBeginOverlap(OverlapSphere)
  (params (OverlappedComponent :type Primitive%20Component%20Object%20Reference) (OtherActor :type Actor%20Object%20Reference) (OtherComp :type Primitive%20Component%20Object%20Reference) (OtherBodyIndex :type Integer) (bFromSweep :type Boolean) (SweepResult :type Hit%20Result%20Structure%20%28by%20ref%29))
  ;[1 1 1] Whenever you overlap a cornfield rectangle, you get this GE unless you already have it
  (bind _asplayer_pawn_athena (Utilities|Casting|CastToPlayerPawn_Athena OtherActor)
    (:then
      (if (Variables|Default|GetShouldWobble)
        (Networking|FlushNetDormancy)
        (CallFunction|MaterialWobble _asplayer_pawn_athena)
        (else
          _))
      (GameplayEffects|ApplyGameplayEffectToTarget :self (Ability|GetAbilitySystemComponent _asplayer_pawn_athena) :GameplayEffectClass (Variables|Default|GetGE_Cornfield_GC) :Target (Ability|GetAbilitySystemComponent _asplayer_pawn_athena) :Level 1.0)
      (Networking|ForceNetUpdate _asplayer_pawn_athena)
      (Variables|/Script/FortniteGame.FortPawn|SetbIgnoreNextFallingDamage true _asplayer_pawn_athena))
    (:CastFailed
      ;[1 1 1] Destroy when cornfield comes in contact with a vehicle
      (Utilities|Casting|CastToFortVehicleInterface OtherActor)
      (Actor|DestroyActor))))

(event OnComponentEndOverlap(OverlapSphere)
  (params (OverlappedComponent :type Primitive%20Component%20Object%20Reference) (OtherActor :type Actor%20Object%20Reference) (OtherComp :type Primitive%20Component%20Object%20Reference) (OtherBodyIndex :type Integer))
  ;[1 1 1] Whenever you stop overlapping a cornfield, remove the GE and the gameplay cue
  (bind _asplayer_pawn_athena (Utilities|Casting|CastToPlayerPawn_Athena OtherActor))
  (GameplayEffects|RemoveActiveGameplayEffectbySourceEffect (Ability|GetAbilitySystemComponent _asplayer_pawn_athena) (Variables|Default|GetGE_Cornfield_GC) 0 1)
  (Networking|ForceNetUpdate _asplayer_pawn_athena))

(fn ConstructionScript)
