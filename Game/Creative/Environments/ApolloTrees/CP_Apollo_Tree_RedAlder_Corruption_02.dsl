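; Corrupted Red Alder tree Blueprint, variant 02, that plays a recolored break effect when destroyed.
; On death it spawns the standard Red Alder break particle emitter at the actor's location, rotation and scale.
; It then sets color parameters through dynamic material instances on the emitter, tinting canopy gradient and falling leaves in red-orange tones instead of using a separate particle asset.
; The construction script only calls its parent.
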(event EventOnDeathPlayEffects
  (params (Damage :type Float%20%28single-precision%29) (DamageTags :type Gameplay%20Tag%20Container%20Structure%20%28by%20ref%29) (Momentum :type Vector :default "0, 0, 0") (HitInfo :type Hit%20Result%20Structure%20%28by%20ref%29) (InstigatedBy :type Fort%20Pawn%20Object%20Reference) (DamageCauser :type Actor%20Object%20Reference) (EffectContext :type Gameplay%20Effect%20Context%20Handle%20Structure))
  (bind _returnvalue (Effects|SpawnEmitteratLocation "/Game/Effects/Fort_Effects/World/Destruction/Trees/TreeDestruction_New/P_RedAlder_Break.P_RedAlder_Break" (Transformation|GetActorLocation) (Transformation|GetActorRotation) (Transformation|GetActorScale3D)))
  ;[1 1 1] Recolor canopy
  (Rendering|Material|SetVectorParameterValue (Material|GetMIDfromMaterial _returnvalue 0 "/Game/Effects/Fort_Effects/World/Destruction/Trees/TreeDestruction_New/MI_RedAlder_leaves_DestVFX.MI_RedAlder_leaves_DestVFX") "Gradient_TopColor" "(R=1.000000,G=0.200000,B=0.000000,A=1.000000)" :!class /Script/Engine.MaterialInstanceDynamic)
  (Rendering|Material|SetVectorParameterValue (Material|GetMIDfromMaterial _returnvalue 0 "/Game/Effects/Fort_Effects/World/Destruction/Trees/TreeDestruction_New/MI_RedAlder_leaves_DestVFX.MI_RedAlder_leaves_DestVFX") "Gradient_BottomColor" "(R=1.000000,G=0.200000,B=0.000000,A=1.000000)" :!class /Script/Engine.MaterialInstanceDynamic)
  ;[1 1 1] Recolor falling leaves
  (Rendering|Material|SetVectorParameterValue (Material|GetMIDfromMaterial _returnvalue 1 "/Game/Effects/Fort_Effects/Materials/Plants/MI_Leaf_CustomColor.MI_Leaf_CustomColor") "ColorHigh" "(R=1.000000,G=0.466667,B=0.000000,A=1.000000)" :!class /Script/Engine.MaterialInstanceDynamic)
  (Rendering|Material|SetVectorParameterValue (Material|GetMIDfromMaterial _returnvalue 1 "/Game/Effects/Fort_Effects/Materials/Plants/MI_Leaf_CustomColor.MI_Leaf_CustomColor") "Colorlow" "(R=0.054902,G=0.000000,B=0.000000,A=1.000000)" :!class /Script/Engine.MaterialInstanceDynamic))

(fn ConstructionScript
  (|CallParentFunction))
