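; Parent Blueprint for Apollo trees in Creative that applies configurable materials to the tree's static mesh.
; On BeginPlay, cosmetic-capable instances async-load the soft Material into slot 1, then load Material_2 into slot 4 if its reference is valid.
; Loaded materials are cached in variables, which the construction script reapplies to both slots.
; GetOverrideMeshMaterial ignores its parameters, sets slot 1 from the cached material and returns 0.
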(event EventBeginPlay
  (Networking|CanExecuteCosmeticEvents
    (:True
      (CallFunction|UpdateVisualOnClient))))

(event Custom|Update%20Visual%20on%20Client
  (bind _material_2 (Variables|Default|GetMaterial_2))
  ;[0.3474 1 0.9102] Initialization
  (bind _object (Utilities|AsyncLoadAsset (Variables|Default|GetMaterial))
    (:Completed
      (bind _asmaterial_interface (Utilities|Casting|CastToMaterialInterface _object))
      (bind _output_get (Variables|Default|SetLoadedMaterial _asmaterial_interface))
      (Rendering|Material|SetMaterial (Variables|Debug|GetStaticMeshComponent) 1 _output_get)
      (if (Utilities|IsValidSoftObjectReference _material_2)
        (bind _object_2 (Utilities|AsyncLoadAsset _material_2)
          (:Completed
            (bind _asmaterial_interface_2 (Utilities|Casting|CastToMaterialInterface _object_2))
            (bind _output_get_2 (Variables|Default|SetLoadedMaterial_2 _asmaterial_interface_2))
            (Rendering|Material|SetMaterial (Variables|Debug|GetStaticMeshComponent) 4 _output_get_2)))))))

(fn GetOverrideMeshMaterial
  (params (MeshComp :type Static%20Mesh%20Component%20Object%20Reference) (MatIdx :type Integer))
  (Rendering|Material|SetMaterial (Variables|Debug|GetStaticMeshComponent) 1 (Variables|Default|GetLoadedMaterial))
  (return 0))

(fn ConstructionScript
  (|CallParentFunction)
  (Rendering|Material|SetMaterial (Variables|Debug|GetStaticMeshComponent) 1 (Variables|Default|GetLoadedMaterial))
  (Rendering|Material|SetMaterial (Variables|Debug|GetStaticMeshComponent) 4 (Variables|Default|GetLoadedMaterial_2)))
