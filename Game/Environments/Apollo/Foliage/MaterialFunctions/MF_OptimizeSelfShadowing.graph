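# Foliage material function that selects between full and optimized self-shadowing inputs by material quality level.
# When the OptimizeSelfShadowing static switch is on, a Quality Switch passes the Enabled input at Epic and the Optimized input at other levels.
# When the switch is off, the Enabled input always passes through.
# Both inputs are supplied by the calling material, so the function holds no shadowing math itself.
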MaterialFunction /Game/Environments/Apollo/Foliage/MaterialFunctions/MF_OptimizeSelfShadowing
graph {nodes [
  n1 = FunctionInput() {inputName=Optimized, sortPriority=3}
  n2 = FunctionInput() {inputName=Enabled}
  FunctionOutput(None=StaticSwitchParameter(True=QualitySwitch(Default=@n1, Epic=@n2), False=@n2) {group=Optimization, parameterName=OptimizeSelfShadowing})
 ]}
