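# Master material for Creative Apollo trees, covering masked two-sided subsurface leaves, trunks and opaque canopies.
# Leaves get sun-driven gradients, position-based color variation and self-shadowing from vertex-color spherical harmonics.
# Trunks add detail bark, ground blending and optional landscape-height snapping, with static switches and a UV mask selecting or blending both layers.
# Chained functions then add snow, distance fades, wind animation, emissive, damage, fresnel and Creative effects.
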Material /Game/Creative/Environments/ApolloTrees/Materials/CP_M_Apollo_Tree_Master
graph {
 properties {
  bUseMaterialAttributes=true
  bUsedWithInstancedStaticMeshes=true
  bUsedWithNanite=true
  blendMode=BLEND_Masked
  ditheredLODTransition=true
  pixelDepthOffsetMode=PDOM_Legacy
  shadingModel=MSM_Subsurface
  twoSided=true
  usedShadingModels=Subsurface
 }
 nodes [
  n1 = WorldPosition() {worldPositionShaderOffset=WPT_ExcludeAllShaderOffsets}
  n2 = CameraVectorWS()
  n3 = Add(
   A=TransformPosition(None=VectorParameter() {group=Emissive, parameterName=EmissiveCenterPosition, defaultValue=<rgba 0 0 0 1>}.RGB) {transformType=TRANSFORMPOSSOURCE_World}
   B=Multiply(A=@n2, B=ScalarParameter() {group=Emissive, parameterName=EmissiveCenterSoften})
  )
  n4 = PixelNormalWS()
  n5 = ScalarParameter() {parameterName=EmissivePixelNormalDistort}
  n6 = Normalize(VectorInput=Add(A=Subtract(A=@n1.XYZ, B=@n3), B=Multiply(A=@n4, B=@n5)))
  n7 = Multiply(
   A=Power(Base=DotProduct(A=@n6, B=@n2), Exp=ScalarParameter() {defaultValue=3, group=Emissive, parameterName=EmissiveFresnel_InterirorFalloff})
   B=VectorParameter() {group=Emissive, parameterName=EmissiveFresnel_CoreBrightness, defaultValue=<rgba 2 1.1601 0 1>}.RGB
  )
  n8 = Fresnel(ExponentIn=ScalarParameter() {defaultValue=3, group=Emissive, parameterName=EmissiveFresnel_Edge}, Normal=@n4) {baseReflectFraction=0}
  n9 = OneMinus(None=@n8)
  n10 = ScalarParameter() {group=Emissive, parameterName=EmissiveFresnelEdgeInvert}
  Add(
   A=@n7
   B=Multiply(
    A=VectorParameter() {group=Emissive, parameterName=EmissiveFresnel_EdgeBrightness, defaultValue=<rgba 2 1.1601 0 1>}.RGB
    B=LinearInterpolate(A=@n8, B=@n9, Alpha=@n10)
   )
  )
  n11 = ActorPositionWS()
  n12 = MaterialFunctionCall(Float3=Multiply(A=@n11, B=ScalarParameter() {defaultValue=0.0012, group=Color, parameterName=Color Variation Range WS, sortPriority=10}) {constB=0.0012}) {materialFunction=&BreakOutFloat3Components}
  n13 = Add(A=Multiply(A=@n12.R) {constB=0.5}, B=Multiply(A=@n12.G) {constB=0.5})
  n14 = Frac(None=@n12.B)
  Add(A=Multiply(A=Subtract(A=@n13) {constB=0.5}) {constB=2}, B=Multiply(A=Subtract(A=@n14) {constB=0.5}) {constB=2})
  Constant3Vector()
  Constant() {r=1}
  Fresnel(ExponentIn=ScalarParameter() {defaultValue=3, group=Emissive, parameterName=EmissiveFresnel_Core}, Normal=PixelNormalWS()) {baseReflectFraction=0}
  n15 = VertexColor()
  n16 = MaterialFunctionCall(X=@n15.G, Y=@n15.G, Z=@n15.G) {materialFunction=&MakeFloat3}
  MaterialExpressionAppendVector_2 = AppendVector(A=Add(A=Multiply(A=@n15.R) {constB=2}), B=Add(A=Multiply(A=@n16.Result) {constB=2}) {constB=-1})
  n17 = CollectionParameter() {
   collection=&FortniteMaterialParameters
   parameterName=SunlightDirectionalVector
  }
  n18 = ComponentMask(None=@n17) {b=true, g=true, r=true}
  MaterialExpressionTransform_0 = Transform(None=@n18) {transformSourceType=TRANSFORMSOURCE_World, transformType=TRANSFORM_Local}
  MaterialExpressionConstant_1 = Constant() {r=1}
  MaterialExpressionConstant_2 = Constant()
  n19 = Custom!(
   SH=@MaterialExpressionAppendVector_2
   Direction=@MaterialExpressionTransform_0
   Scale=@MaterialExpressionConstant_1
   Offset=@MaterialExpressionConstant_2
  ) {
   code="//SH decode,  Direction = LightDir, SH = 2band coefficients. (encoded in Y-up)\r\nDirection = normalize(Direction);\r\nfloat4 shEvalDirection;\r\nshEvalDirection.x = 0.282095;\r\nshEvalDirection.y =0.488603 * Direction.z;\r\nshEvalDirection.z = 0.488603 * Direction.y;\r\nshEvalDirection.w =-0.488603 * Direction.x;\r\nreturn saturate(dot(SH, shEvalDirection));"
   description=SphericalHarmonicsEval
   outputType=CMOT_Float1
   inputs [
    {inputName=SH, input {expression=&self:MaterialExpressionAppendVector_2}}
    {
     inputName=Direction
     input {expression=&self:MaterialExpressionTransform_0}
    }
    {inputName=Scale, input {expression=&self:MaterialExpressionConstant_1}}
    {inputName=Offset, input {expression=&self:MaterialExpressionConstant_2}}
   ]
  }
  n20 = OneMinus(None=@n19)
  n21 = Multiply(A=@n20) {constB=4}
  n22 = CollectionParameter() {collection=&FortniteMaterialParameters, parameterName=ShadowsEnabled}
  n23 = Saturate(None=Add(A=Add(A=@n21) {constB=-2}, B=@n22))
  n24 = ShadingPathSwitch(Default=@n23, Mobile=QualitySwitch(Default=@n23, Low=Constant() {r=1})) {desc=Migrated from FeatureLevelSwitch}
  n25 = ScalarParameter() {defaultValue=0.375, group=Shading Leafs, parameterName=DiffuseBrightness}
  n26 = ScalarParameter() {defaultValue=1.25, group=Shading Leafs, parameterName=SubSurface_SSS_Brightness}
  n27 = Add(A=@n26, B=@n25)
  n28 = ShadingPathSwitch(Default=@n25, Deferred=@n25, Forward=@n27, Mobile=@n27)
  n29 = MaterialFunctionCall(Float3=@n11) {materialFunction=&BreakOutFloat3Components}
  n30 = ScalarParameter() {defaultValue=13000, group=Color, parameterName=Color Variation Distance WS, sortPriority=10}
  n31 = ScalarParameter() {defaultValue=11000, group=Color, parameterName=Color Variation Distance WS2, sortPriority=10}
  n32 = Add(A=Sine(None=Divide(A=@n29.R, B=@n30)), B=Sine(None=Divide(A=@n29.G, B=@n31)))
  n33 = ComponentMask(None=VertexInterpolator(VS=Divide(A=@n32)).PS) {r=true}
  n34 = ScalarParameter() {defaultValue=0.5, group=Color, parameterName=ValueVariation}
  n35 = ScalarParameter() {defaultValue=-0.03, group=Color, parameterName=Hue Variation}
  n36 = VectorParameter() {group=Color, parameterName=Color1_Base, defaultValue=<rgba 0.2016 0.2195 0.0423 1>}
  n37 = Power(Base=@n20, Exp=ScalarParameter() {defaultValue=2, group=ShadowGradient, parameterName=ShadowGradientContrast, sliderMax=5, sliderMin=0.5})
  n38 = ScalarParameter() {defaultValue=2, group=ShadowGradient, parameterName=ShadowGradientIntensity, sliderMax=5, sliderMin=0.5}
  n39 = Normalize(VectorInput=Subtract(A=WorldPosition().XYZ, B=ObjectPositionWS()))
  n40 = VertexNormalWS()
  n41 = Saturate(None=DotProduct(A=@n18, B=StaticSwitchParameter(True=@n39, False=@n40) {group=ShadowGradient, parameterName=ColorGradientWorldSpace}))
  n42 = ShadingPathSwitch(Default=Saturate(None=Multiply(A=@n37, B=@n38)), Mobile=@n41) {desc=Migrated from FeatureLevelSwitch}
  n43 = VertexInterpolator(VS=@n42)
  n44 = Subtract(A=@n43.PS, B=Constant() {r=0.8})
  n45 = Constant() {r=0.5}
  n46 = VertexInterpolator(VS=Saturate(None=Subtract(A=@n41, B=@n45)))
  n47 = StaticSwitchParameter(
   True=MaterialFunctionCall!(Enabled=Saturate(None=@n44), Optimized=@n46.PS) {materialFunction=/Game/Environments/Apollo/Foliage/MaterialFunctions/MF_OptimizeSelfShadowing}.Result
   False=@n46.PS
  ) {group=ShadowGradient, parameterName=UseSelfShadows}
  n48 = ComponentMask(None=ShadingPathSwitch(Default=@n47, Deferred=@n47, Forward=@n46.PS, Mobile=@n46.PS)) {r=true}
  n49 = Saturate(None=@n43.PS)
  n50 = Subtract(A=@n41, B=Constant() {r=-0.5})
  n51 = VertexInterpolator(VS=OneMinus(None=Saturate(None=@n50)))
  n52 = StaticSwitchParameter(
   True=MaterialFunctionCall(Enabled=OneMinus(None=@n49), Optimized=@n51.PS) {materialFunction=&MF_OptimizeSelfShadowing}.Result
   False=@n51.PS
  ) {group=ShadowGradient, parameterName=UseSelfShadows}
  n53 = ComponentMask(None=ShadingPathSwitch(Default=@n52, Deferred=@n52, Forward=@n51.PS, Mobile=@n51.PS)) {r=true}
  n54 = LinearInterpolate(
   A=LinearInterpolate(A=VectorParameter() {group=Color, parameterName=Color2_Lit, defaultValue=<rgba 0.0887 0.1046 0.0152 1>}.RGB, B=@n36.RGB, Alpha=@n48)
   B=VectorParameter() {group=Color, parameterName=Color3_Shadows, defaultValue=<rgba 0.0356 0.0612 0.0116 1>}.RGB
   Alpha=@n53
  )
  n55 = Saturate(None=DotProduct(A=CameraVectorWS(), B=Multiply(A=@n18) {constB=-1}))
  n56 = Multiply(A=@n55, B=@n55)
  n57 = Multiply(
   A=Add(A=Multiply(A=@n33, B=@n34))
   B=MaterialFunctionCall(Hue Shift Percentage=Multiply(A=@n33, B=@n35), Texture=LinearInterpolate(A=@n54, B=@n36.RGB, Alpha=@n56)) {materialFunction=&HueShift}.Result
  )
  n58 = ShadingPathSwitch(Default=@n57, Deferred=@n57, Forward=@n36.RGB, Mobile=@n36.RGB)
  n59 = TextureSampleParameter2D() {parameterName=MaskTexture_OpaqueCanopy, samplerSource=SSM_Wrap_WorldGroupSettings, texture=&T_WhitesRGB}
  n60 = TextureSampleParameter2D() {parameterName=MaskTexture, samplerSource=SSM_Wrap_WorldGroupSettings, texture=&T_WhitesRGB}
  n61 = Constant() {r=1}
  n62 = If(A=ComponentMask(None=TextureCoordinate()) {r=true}, B=Constant() {r=0.995}, "A > B"=Constant(), "A == B"=@n61, "A < B"=@n61)
  n63 = LinearInterpolate(A=@n59.RGB, B=@n60.RGB, Alpha=@n62)
  n64 = MaterialFunctionCall(Float3=StaticSwitchParameter(True=StaticSwitchParameter(True=@n59.RGB, False=@n63) {group=Trunk, parameterName=IsTrunk}, False=@n60.RGB) {parameterName=UseOpaqueCanopy}) {materialFunction=&BreakOutFloat3Components}
  n65 = Multiply(A=@n58, B=@n64.B)
  n66 = Subtract(A=Add(A=Multiply(A=@n65) {constB=2}), B=@n64.B)
  n67 = TextureSampleParameter2D() {
   parameterName=Diffuse_OpaqueCanopy
   samplerSource=SSM_Wrap_WorldGroupSettings
   samplerType=SAMPLERTYPE_Masks
   texture=/Game/Global/Textures/Default/Blanks/T_MidgreyMask
  }
  n68 = TextureSampleParameter2D() {
   parameterName=Diffuse
   samplerSource=SSM_Wrap_WorldGroupSettings
   samplerType=SAMPLERTYPE_Masks
   texture=&T_MidgreyMask
  }
  n69 = StaticSwitchParameter(
   True=StaticSwitchParameter(True=@n67.RGB, False=LinearInterpolate(A=@n67.RGB, B=@n68.RGB, Alpha=@n62)) {group=Trunk, parameterName=IsTrunk}
   False=@n68.RGB
  ) {parameterName=UseOpaqueCanopy}
  n70 = StaticBool()
  n71 = StaticSwitchParameter(True=StaticSwitch(True=@n69, False=Multiply(A=@n66, B=@n69), Value=@n70), False=@n58) {defaultValue=true, group=Color, parameterName=UseDiffuse, sortPriority=1}
  n72 = ScalarParameter() {defaultValue=0.6, parameterName=Canopy_AO_Intensity}
  n73 = LinearInterpolate(A=Saturate(None=Add(A=@n67.A, B=@n72)), Alpha=@n62)
  n74 = ComponentMask(None=StaticSwitchParameter(True=Multiply(A=@n71, B=@n73), False=@n71) {parameterName=UseOpaqueCanopy}) {b=true, g=true, r=true}
  n75 = Subtract(A=WorldPosition().XYZ, B=ActorPositionWS())
  n76 = Saturate(None=Divide(A=ComponentMask(None=@n75) {b=true}, B=ScalarParameter() {defaultValue=491.7252, group=WS GRadient, parameterName=WSGraundGradient}))
  n77 = VectorParameter() {parameterName=GroundBlendingColor, defaultValue=<rgba 0.1302 0.0805 0.0274 1>}
  n78 = StaticSwitchParameter(
   True=StaticSwitchParameter(True=LinearInterpolate(A=@n77.RGB, B=@n74, Alpha=@n76), False=@n77.RGB) {parameterName=UseGrassColorsGradient}
   False=@n74
  ) {parameterName=UseGrassColors}
  n79 = StaticSwitchParameter(True=@n42, False=Multiply(A=StaticSwitchParameter(True=@n27, False=@n28) {group=ShadowGradient, parameterName=DebugMOBILESSSOFF}, B=@n78)) {group=ShadowGradient, parameterName=DebugSelfShadows}
  n80 = Multiply(A=@n24, B=@n79)
  n81 = ScalarParameter() {defaultValue=0.125, group=Shading Leafs, parameterName=Specular_Opaque_Canopy}
  n82 = ScalarParameter() {defaultValue=0.25, group=Shading Leafs, parameterName=Specular_Leafs_Shadow}
  n83 = ScalarParameter() {defaultValue=0.25, group=Shading Leafs, parameterName=Specular_Leafs}
  n84 = Power(
   Base=@n20
   Exp=ScalarParameter() {defaultValue=4, group=ShadowGradient, parameterName=ShadowGradientSpecularContrast, sliderMax=5, sliderMin=0.5}
  )
  n85 = LinearInterpolate(A=@n82, B=@n83, Alpha=@n84)
  n86 = ComponentMask(None=VertexInterpolator(VS=@n85).PS) {r=true}
  n87 = VertexInterpolator(VS=@n41)
  n88 = LinearInterpolate(A=@n82, B=@n83, Alpha=@n87.PS)
  n89 = MaterialFunctionCall(Enabled=StaticSwitchParameter(True=@n86, False=@n85) {defaultValue=true, parameterName=UseFlipBackFaceNormals}, Optimized=@n88) {materialFunction=&MF_OptimizeSelfShadowing}
  n90 = Saturate(None=StaticSwitchParameter(True=ShadingPathSwitch(Default=@n89.Result, Deferred=@n89.Result, Forward=@n88, Mobile=@n88), False=@n88) {group=ShadowGradient, parameterName=UseSelfShadows})
  n91 = StaticSwitchParameter(True=LinearInterpolate(A=Multiply(A=@n73, B=@n81), B=@n90, Alpha=@n62), False=@n90) {parameterName=UseOpaqueCanopy}
  n92 = Multiply(A=@n24, B=@n91)
  n93 = Multiply(
   A=@n78
   B=LinearInterpolate(A=@n26, B=ScalarParameter() {defaultValue=0.0887, group=Shading Leafs, parameterName=SubSurface_SSS_Brightness_Shadows}, Alpha=@n55)
  )
  n94 = Power(Base=OneMinus(None=@n19), Exp=ScalarParameter() {defaultValue=1.5, group=Shading Leafs, parameterName=SubSurface_SH_Contrast_EPIC})
  n95 = StaticSwitchParameter(True=QualitySwitch(Default=@n93, Epic=Multiply(A=@n93, B=@n94)), False=@n93) {group=ShadowGradient, parameterName=UseSelfShadows}
  n96 = StaticSwitchParameter(True=LinearInterpolate(A=Constant3Vector(), B=@n95, Alpha=VertexColor().G), False=@n95) {parameterName=UseVertexcolormaskforSImpleAnimation}
  n97 = LinearInterpolate(
   A=ScalarParameter() {defaultValue=0.03, group=Shading Leafs, parameterName=SubSurface_SSS_Opacity}
   B=ScalarParameter() {defaultValue=0.0395, group=Shading Leafs, parameterName=SubSurface_SSS_Opacity2}
   Alpha=@n55
  )
  n98 = LinearInterpolate(
   A=ScalarParameter() {group=Shading Leafs, parameterName="SubSurface_SSS_Opacity_VertexColormask(greenchannel)"}
   B=@n97
   Alpha=VertexColor().G
  )
  n99 = TextureSampleParameter2D() {
   parameterName=Normals_OpaqueCanopy
   samplerSource=SSM_Wrap_WorldGroupSettings
   samplerType=SAMPLERTYPE_Normal
   texture=/Game/Environments/Apollo/Foliage/Textures_Materials/T_Oak_Leaves_N
  }
  n100 = TextureSampleParameter2D() {
   parameterName=Normals
   samplerSource=SSM_Wrap_WorldGroupSettings
   samplerType=SAMPLERTYPE_Normal
   texture=&T_Oak_Leaves_N
  }
  n101 = StaticSwitchParameter(
   True=StaticSwitchParameter(True=@n99.RGB, False=LinearInterpolate(A=@n99.RGB, B=@n100.RGB, Alpha=@n62)) {group=Trunk, parameterName=IsTrunk}
   False=@n100.RGB
  ) {parameterName=UseOpaqueCanopy}
  n102 = StaticSwitchParameter(True=Multiply(A=@n101, B=TwoSidedSign()), False=@n101) {defaultValue=true, parameterName=UseFlipBackFaceNormals}
  n103 = ScalarParameter() {group=Shading Leafs, parameterName=FlattenNormal_Leafs}
  n104 = MakeMaterialAttributes!(
   BaseColor=StaticSwitchParameter(
    True=MaterialFunctionCall(Enabled=@n80, Optimized=@n79) {materialFunction=&MF_OptimizeSelfShadowing}.Result
    False=@n79
   ) {group=ShadowGradient, parameterName=UseSelfShadows}
   Metallic
   Specular=StaticSwitchParameter(
    True=MaterialFunctionCall(Enabled=@n92, Optimized=@n91) {materialFunction=&MF_OptimizeSelfShadowing}.Result
    False=@n91
   ) {group=ShadowGradient, parameterName=UseSelfShadows}
   Roughness=ScalarParameter() {defaultValue=0.5, group=Shading Leafs, parameterName=Roughness Leafs}
   Anisotropy
   EmissiveColor
   Opacity=StaticSwitchParameter(True=StaticSwitchParameter(True=@n98, False=@n97) {parameterName=UseVertexcolormaskforSImpleAnimation}, False=@n97) {parameterName=UseSimpleNonePivotPainterAnimation}
   OpacityMask=@n64.R
   Normal=StaticSwitch(
    True=@n102
    False=MaterialFunctionCall(Normal=@n102, Flatness=@n103) {materialFunction=&FlattenNormal}.Result
    Value=@n70
   )
   Tangent
   WorldPositionOffset
   SubsurfaceColor=StaticSwitchParameter(True=Constant(), False=StaticSwitchParameter(True=@n96, False=@n95) {parameterName=UseSimpleNonePivotPainterAnimation}) {group=ShadowGradient, parameterName=DebugMOBILESSSOFF}
   ClearCoat
   ClearCoatRoughness
   AmbientOcclusion
   Refraction
   CustomizedUVs_0
   CustomizedUVs_1
   CustomizedUVs_2
   CustomizedUVs_3
   CustomizedUVs_4
   CustomizedUVs_5
   CustomizedUVs_6
   CustomizedUVs_7
   PixelDepthOffset
   ShadingModel
   Displacement
  )
  n105 = TextureSampleParameter2D() {
   group=Trunk
   parameterName=Trunk_BaseColor
   samplerSource=SSM_Wrap_WorldGroupSettings
   texture=/Game/Global/Textures/Default/Blanks/T_BlackUserInterface
  }
  n106 = Subtract(A=WorldPosition().XYZ, B=Add(A=ActorPositionWS(), B=ScalarParameter() {group=Trunk, parameterName=Ground Blending Offset}))
  n107 = ScalarParameter() {defaultValue=256, group=Trunk, parameterName=Ground Blending Fade Out Range}
  n108 = Saturate(None=Divide(A=ComponentMask(None=@n106) {b=true}, B=Multiply(A=@n107) {constB=2, desc=why}))
  n109 = Multiply(
   A=ComponentMask(None=VectorParameter() {group=Trunk, parameterName=Detail Tiling, sortPriority=50, defaultValue=<rgba 4 2 0 1>}.RGB) {g=true, r=true}
   B=TextureCoordinate()
  )
  n110 = TextureSampleParameter2D(UVs=@n109) {
   group=Trunk
   parameterName=Trunk_DetailMap
   samplerSource=SSM_Wrap_WorldGroupSettings
   samplerType=SAMPLERTYPE_Masks
   texture=/Game/Environments/Apollo/Foliage/Textures_Materials/Apollo_Bark_Detail_D_01
  }
  n111 = ComponentMask(None=@n110.RGB) {r=true}
  n112 = Multiply(A=ScalarParameter() {defaultValue=1.0272, parameterName=Detail Strength}, B=Subtract(A=Multiply(A=@n111) {constB=2}))
  n113 = Multiply(
   A=StaticSwitchParameter(True=LinearInterpolate(A=@n77.RGB, B=@n105.RGB, Alpha=@n108), False=@n105.RGB) {defaultValue=true, parameterName=UseGroundBlend}
   B=Add(A=@n112)
  )
  n114 = Saturate(None=Multiply(A=@n112) {constB=-1})
  n115 = TextureSampleParameter2D() {
   group=Trunk
   parameterName=Trunk_Specular
   samplerSource=SSM_Wrap_WorldGroupSettings
   samplerType=SAMPLERTYPE_Masks
   texture=&T_Fortnite_Default_S
  }
  n116 = LinearInterpolate(
   A=ScalarParameter() {defaultValue=-1, parameterName=Trunk_Diffuse_spec_high}
   B=ScalarParameter() {defaultValue=4, parameterName=Trunk_Diffuse_spec_low}
   Alpha=@n105.R
  )
  n117 = QualitySwitch(Default=@n115.R, Epic=Multiply(A=@n115.R, B=Saturate(None=@n116)))
  n118 = LinearInterpolate(A=Constant() {r=1}, B=@n115.B, Alpha=@n108)
  n119 = TextureSampleParameter2D() {
   group=Trunk
   parameterName=Trunk_Normal
   samplerSource=SSM_Wrap_WorldGroupSettings
   samplerType=SAMPLERTYPE_Normal
   texture=&T_Oak_Leaves_N
  }
  n120 = TextureSampleParameter2D(UVs=@n109) {
   group=Trunk
   parameterName=Trunk_Detail_Normal
   samplerSource=SSM_Wrap_WorldGroupSettings
   samplerType=SAMPLERTYPE_Normal
   texture=/Game/Environments/Apollo/Foliage/Textures_Materials/Apollo_Bark_Detail_N_01
  }
  n121 = QualitySwitch(
   Default=@n120.RGB
   Epic=MaterialFunctionCall(Normal=@n120.RGB, Flatness=ScalarParameter() {defaultValue=-2, parameterName=BarkDetailFlattnenorma}) {materialFunction=&FlattenNormal}.Result
  )
  n122 = WorldPosition() {worldPositionShaderOffset=WPT_ExcludeAllShaderOffsets}
  n123 = ComponentMask(None=@n122.XYZ) {b=true}
  n124 = RuntimeVirtualTextureSample!(UVs, World Position) {materialType=WorldHeight, virtualTexture=/Game/Athena/Apollo/Maps/Landscape/VT/RVT_Apollo_Landscape_Height}
  n125 = Subtract(
   A=Min(A=Subtract(A=@n123, B=@n124.WorldHeight)) {constB=200}
   B=ComponentMask(None=MaterialFunctionCall() {materialFunction=&LocalPosition}."Local Position (Excluding Offsets)") {b=true}
  )
  n126 = ActorPositionWS()
  n127 = ScalarParameter() {defaultValue=50, group=Trunk, parameterName=Ground Snapping Offset}
  n128 = ComponentMask(None=Subtract(A=@n122.XYZ, B=Add(A=@n126, B=@n127))) {b=true}
  n129 = Multiply(A=ScalarParameter() {defaultValue=500, group=Trunk, parameterName=Ground Snapping Fade Out Range}) {constB=2}
  n130 = MaterialFunctionCall(
   Transition Phase=Saturate(None=Divide(A=@n128, B=@n129))
   Height Texture=Constant() {r=1}
   Contrast=ScalarParameter() {group=Trunk, parameterName=Ground Snapping Fade Out Constrast}
  ) {materialFunction=&HeightLerp}
  n131 = AppendVector(A=Constant2Vector(), B=Multiply(A=Multiply(A=@n125) {constB=-1}, B=OneMinus(None=@n130.Alpha)))
  n132 = ActorPositionWS()
  n133 = ComponentMask(None=@n132) {b=true}
  n134 = Subtract(
   A=ComponentMask(None=@n132) {g=true, r=true}
   B=ComponentMask(None=Constant3Vector() {constant=<rgba -5602.9888 -5602.9399 0 0>}) {g=true, r=true}
  ) {constB=2.9888}
  n135 = Constant() {r=0.0088}
  n136 = TextureSample(UVs=Multiply(A=Multiply(A=@n134, B=@n135) {constB=0.0088}) {constB=0.01}) {
   samplerSource=SSM_Wrap_WorldGroupSettings
   samplerType=SAMPLERTYPE_LinearGrayscale
   texture=/Game/Environments/Apollo/Foliage/Textures_Materials/Apollo_Test_Heightmap_VS
  }
  n137 = Subtract(A=Multiply(A=@n136.R) {constB=51200}) {constB=25600}
  n138 = Multiply(A=Abs(None=Subtract(A=@n133, B=@n137))) {constB=0.01}
  n139 = OneMinus(None=Floor(None=Saturate(None=@n138)))
  n140 = ComponentMask(None=WorldPosition().XYZ) {b=true}
  n141 = MakeMaterialAttributes(
   BaseColor=QualitySwitch(Default=Saturate(None=@n113), Low=@n105.RGB, Medium=@n105.RGB)
   Specular=QualitySwitch(Default=Subtract(A=@n117, B=@n114), Low=@n117, Medium=@n117)
   Roughness=QualitySwitch(Default=StaticSwitchParameter(True=@n118, False=@n115.B) {defaultValue=true, parameterName=UseGroundBlend}, Low=@n115.B, Medium=@n115.B)
   Opacity=StaticSwitchParameter(True=Constant() {r=1}, False=Constant() {desc=So nobody forgets to set it to opaque}) {parameterName=UseOpaqueCanopy}
   OpacityMask=Sine(None=Divide(A=@n140) {constB=50})
   Normal=QualitySwitch(
    Default=MaterialFunctionCall(BaseNormal=@n119.RGB, AdditionalNormal=@n121) {materialFunction=&BlendAngleCorrectedNormals}.Result
    Low=@n119.RGB
    Medium=@n119.RGB
   )
   WorldPositionOffset=StaticSwitchParameter(True=Multiply(A=@n131, B=@n139), False=Constant()) {group=Trunk, parameterName=Use Landscape Base Snapping}
   SubsurfaceColor=StaticSwitchParameter(True=Constant3Vector(), False=Constant3Vector() {constant=<rgba 8 0.95 0 0>}) {parameterName=UseOpaqueCanopy}
  )
  n142 = BlendMaterialAttributes(A=@n104, B=@n141, Alpha=@n62)
  n143 = TextureSampleParameter2D() {
   parameterName=Specular
   samplerSource=SSM_Wrap_WorldGroupSettings
   samplerType=SAMPLERTYPE_Masks
   texture=&T_MidgreyMask
  }
  n144 = SetMaterialAttributes!(MaterialAttributes=@n104, Specular=@n143.R, Metallic=@n143.R, Roughness=@n143.R) {attributeSetTypes [#1, #2, #3]}
  n145 = MaterialFunctionCall!(
   In=StaticSwitchParameter(
    True=StaticSwitchParameter(True=@n142, False=@n141) {parameterName=UseOpaqueCanopy}
    False=StaticSwitchParameter(True=@n144, False=@n104) {parameterName=UseSpecularTexture}
   ) {group=Trunk, parameterName=IsTrunk}
   FlipTwoSidedSign
   IsTangentSpaceInput=StaticBool() {value=true}
   HasBackFlipFaceNormals
  ) {materialFunction=/Game/Packages/Fortnite_UniqueMaterials_1/MaterialFunctions/Trees/MF_Tree_Snow}
  n146 = MaterialFunctionCall!(
   In=MaterialFunctionCall!(In=MaterialFunctionCall!(In=@n145.Result, HeightModulation=@n64.G) {materialFunction=/Game/Environments/Apollo/Foliage/MaterialFunctions/MF_DistanceMaskFade}.Result) {materialFunction=/Game/Environments/Apollo/Foliage/MaterialFunctions/MF_WPOWobble}.Result
   OpaqueCanopyMask
  ) {materialFunction=/Game/Environments/Apollo/Foliage/MaterialFunctions/MF_TreeAnim_Apollo_Optimized}
  n147 = MaterialFunctionCall!(In=@n146.Result, Mask=@n62) {desc="Used for Opaque Perf optmization ", materialFunction=/Game/Environments/Apollo/Foliage/MaterialFunctions/MF_DistanceWPOfade}
  n148 = TwoSidedSign()
  n149 = Constant() {r=1}
  n150 = StaticSwitch(
   True=MaterialFunctionCall!(
    Material Attributes=@n147.Result
    IsTangentSpaceNormal=StaticBool() {desc="DON\"T CHANGE THIS", value=true}
    In=StaticSwitchParameter(True=@n148, False=@n149) {defaultValue=true, parameterName=UseFlipBackFaceNormals}
   ) {materialFunction=/Game/ImpostorBaker/MaterialFunctions/ImposterCaptureSwitch}.Result
   False=@n147.Result
   Value=@n70
  )
  n151 = GetMaterialAttributes(None=@n150) {attributeGetTypes [#4, #5]}
  n152 = ScalarParameter() {defaultValue=5, parameterName=MobileMaxBrightness}
  n153 = TextureSampleParameter2D() {
   group=Emissive
   parameterName=EmissiveTexture
   samplerSource=SSM_Wrap_WorldGroupSettings
   samplerType=SAMPLERTYPE_Masks
   texture=/Engine/MaterialTemplates/LayerBlends/T_DefaultMaskTexture
  }
  n154 = ChannelMaskParameter(None=@n153.RGB) {maskChannel=Blue, parameterName=EmisiveMask, defaultValue=<rgba 0 0 1 0>}
  n155 = LinearInterpolate(
   A=ScalarParameter() {group=Emissive, parameterName=EmissiveContrast_1}
   B=ScalarParameter() {defaultValue=1, group=Emissive, parameterName=EmissiveContrast_2}
   Alpha=ChannelMaskParameter(None=StaticSwitchParameter(True=@n153.RGB, False=@n151.Base Color) {parameterName=UseEmissiveTexture}) {maskChannel=Blue, parameterName=EmisiveMask, defaultValue=<rgba 0 0 1 0>}.R
  )
  n156 = ScalarParameter() {defaultValue=1, group=Emissive, parameterName=EmissivePower}
  n157 = StaticSwitchParameter(
   True=StaticSwitchParameter(True=Multiply(A=@n154.R, B=@n7), False=@n7) {parameterName=UseEmissiveTexture}
   False=Multiply(A=VectorParameter() {group=Emissive, parameterName=EmissiveColor, defaultValue=<rgba 1 1 1 1>}.RGB, B=Power(Base=@n155, Exp=@n156))
  ) {parameterName=UseEmissiveInterirorFresnel}
  n158 = SetMaterialAttributes!(
   MaterialAttributes=@n151.MaterialAttributes
   Emissive Color=ShadingPathSwitch(Default=@n157, Mobile=Clamp(None=@n157, Max=@n152)) {desc=Migrated from FeatureLevelSwitch}
  ) {attributeSetTypes [#5]}
  n159 = MaterialFunctionCall!(In=StaticSwitchParameter(True=MaterialFunctionCall!(In=@n158) {materialFunction=/Game/Athena/WinterArt/Materials/MF_TOD_EmissiveColors}.Result, False=@n150) {parameterName=UseEmissive}) {materialFunction=/Game/Environments/Props/Winter/Materials/EmissiveTextureWinterTree}
  n160 = StaticBool() {value=true}
  n161 = MaterialFunctionCall!(In=MaterialFunctionCall!(
    In=MaterialFunctionCall(Material_In=@n159.Material Output, UseAdditiveWPO=@n160) {materialFunction=&MF_FN_DestructibleDamage}.Material_Out
    Mask=@n43.PS
    IsImpostor
   ) {materialFunction=/Game/Environments/Apollo/Foliage/MaterialFunctions/MF_TreeFresnel}.Result) {materialFunction=/Game/Environments/Apollo/Foliage/MaterialFunctions/MF_SubSurfaceFadeEpicDetail}
  MaterialFunctionCall!(In=MaterialFunctionCall(MA In=MaterialFunctionCall!(In=@n161.Result, UseSphericalMappingOverride) {materialFunction=/Game/Packages/Fortress_SharedMaterials/Base_Material_Functions/MF_CurieFallbackEffect}.Result) {materialFunction=&MF_CreativeMoveTool}.MA Out) {materialFunction=/Game/Packages/Fortress_SharedMaterials/Cinematics/CINE_MF_EnvPostFX}
  Multiply(
   A=@n84
   B=ScalarParameter() {defaultValue=2, group=ShadowGradient, parameterName=ShadowGradientSpecularIntensity, sliderMax=5, sliderMin=0.5}
  ) {desc=Removed to match Low detail}
  n162 = TwoSidedSign()
  Multiply(A=@n86, B=@n162)
  OneMinus(None=@n33)
  Reroute(None=@n101)
  ScalarParameter() {defaultValue=0.75, group=Shading Leafs, parameterName=Shift_Gradient}
  ScalarParameter() {group=Shading Leafs, parameterName=Shift_Gradient_2}
  n163 = Constant() {r=0.5}
  n164 = VertexInterpolator(VS=Divide(A=Add(A=@n32) {constB=2}) {constB=4})
  n165 = ScalarParameter() {defaultValue=0.5, group=Color, parameterName=Fixed Color Manual Offset, sortPriority=-9}
  n166 = Multiply(A=Subtract(A=ComponentMask(None=@n146.Random Value perleaf) {r=true}) {constB=0.5}) {constB=2}
  n167 = Subtract(A=Distance(A=CameraPositionWS(), B=@n11)) {constB=10000}
  n168 = OneMinus(None=Saturate(None=Multiply(A=@n167) {constB=0.0001}))
  n169 = Multiply(A=ScalarParameter() {defaultValue=0.05, group=Color, parameterName=Color Variation Scale per Leaf, sortPriority=11}, B=@n168)
  n170 = ScalarParameter() {defaultValue=0.05, group=Color, parameterName=Color Variation Scale per Pixel, sortPriority=12}
  n171 = Subtract(A=@n64.G) {constB=0.5}
  n172 = Add(
   A=Add(
    A=StaticSwitchParameter(True=@n164.PS, False=@n165) {group=Color, parameterName=UseRandomColor, sortPriority=-10}
    B=Multiply(A=@n166, B=@n169) {constB=0.5}
   )
   B=Multiply(A=Multiply(A=@n170, B=@n168) {constB=0.5}, B=Multiply(A=@n171) {constB=2}) {constB=0.5}
  )
  n173 = TextureObjectParameter() {group=Color, parameterName=ColorPalette, sortPriority=99, texture=/Game/Environments/Apollo/Foliage/ColorLUT/Apollo_Foliage_LUT_Green}
  n174 = TextureSample(
   UVs=AppendVector(A=ShadingPathSwitch(Default=@n172, Mobile=@n163) {desc=Migrated from FeatureLevelSwitch}, B=Constant() {r=0.375})
   Tex=@n173
  ) {samplerSource=SSM_Wrap_WorldGroupSettings, texture=&Apollo_Foliage_LUT_Green}
  n175 = TextureSample(UVs=AppendVector(A=@n172, B=Constant() {r=0.125}), Tex=@n173) {samplerSource=SSM_Wrap_WorldGroupSettings, texture=&Apollo_Foliage_LUT_Green}
  n176 = AppendVector(A=@n172, B=Constant() {r=0.625})
  ShadingPathSwitch(
   Default=LinearInterpolate(
    A=LinearInterpolate(A=@n174.RGB, B=@n175.RGB, Alpha=@n48)
    B=TextureSample(UVs=@n176, Tex=@n173) {samplerSource=SSM_Wrap_WorldGroupSettings, texture=&Apollo_Foliage_LUT_Green}.RGB
    Alpha=@n53
   )
   Mobile=@n174.RGB
  ) {desc=Migrated from FeatureLevelSwitch}
  StaticSwitchParameter(True=Multiply(A=@n162, B=@n87.PS), False=@n87.PS) {defaultValue=true, parameterName=UseFlipBackFaceNormals}
  n177 = Multiply(A=OneMinus(None=@n97), B=@n94)
  StaticSwitchParameter(True=QualitySwitch(Default=@n97, Epic=OneMinus(None=@n177)), False=@n97) {group=ShadowGradient, parameterName=UseSelfShadows}
  VertexInterpolator()
 ]
}
